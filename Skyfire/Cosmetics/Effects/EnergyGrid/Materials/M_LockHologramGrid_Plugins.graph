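# Translucent, unlit, two-sided hologram material for a lock effect in the EnergyGrid cosmetics, showing a tinted energy grid with blinking dots.
# A squared sine UV mask shapes a soft fill and a panning grid distorted by scrolling noise, while a dot texture's red channel offsets each blink's phase.
# The combined pattern drives colour and a Fresnel-faded alpha scaled by Spawn Animation Time, which a debug switch replaces with a time-based pulse.
# Vertex-offset math and an AlignMeshToTheCamera call are also present.
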Material /Skyfire/Cosmetics/Effects/EnergyGrid/Materials/M_LockHologramGrid_Plugins
graph {
 properties {blendMode=BLEND_Translucent, pixelDepthOffsetMode=PDOM_Legacy, shadingModel=MSM_Unlit, twoSided=true, usedShadingModels=Unlit}
 nodes [
  n1 = VectorParameter() {group=Animation, parameterName="Scale over time (xX yY)", defaultValue=<rgba 0 100 0 125>}
  n2 = Cosine(None=Multiply(A=Time()) {constB=0.3})
  n3 = StaticSwitchParameter(
   True=Saturate(None=Add(A=@n2) {constB=0.5})
   False=ScalarParameter() {defaultValue=1, group=Logic, parameterName=Spawn Animation Time, sliderMax=1}
  ) {group=0_DEBUG, parameterName=DEBUG MODE}
  n4 = OneMinus(None=@n3)
  n5 = LinearInterpolate(A=@n1.B, B=@n1.B, Alpha=@n4)
  n6 = Transform(None=Constant3Vector() {constant=<rgba 0 -1 0 0>})
  n7 = Subtract(A=TextureCoordinate()) {constB=0.5}
  n8 = LinearInterpolate(A=@n1.R, B=@n1.R, Alpha=@n4)
  n9 = Transform(None=Constant3Vector() {constant=<rgba 1 0 0 0>})
  Add(
   A=Multiply(A=Multiply(A=@n5, B=@n6), B=ComponentMask(None=@n7) {g=true})
   B=Multiply(A=Multiply(A=@n8, B=@n9), B=ComponentMask(None=@n7) {r=true})
  )
  MaterialFunctionCall!(
   WorldPosition
   PivotLocation
   WS Normals
   Custom Object Basis 1=Constant3Vector() {constant=<rgba 1 0 0 0>}
   Custom Object Basis 2=Constant3Vector() {constant=<rgba 0 0 1 0>}
   Custom Object Basis 3=Constant3Vector() {constant=<rgba 0 1 0 0>}
  ) {materialFunction=/Engine/Functions/Engine_MaterialFunctions02/WorldPositionOffset/AlignMeshToTheCamera}
  n10 = ScalarParameter() {defaultValue=1, parameterName=ExponentIn}
  n11 = Multiply(A=TextureCoordinate()) {constB=0.5}
  n12 = Multiply(A=Sine(None=ComponentMask(None=@n11) {r=true}), B=Sine(None=ComponentMask(None=@n11) {g=true}))
  n13 = Multiply(A=@n12, B=@n12)
  n14 = Multiply(A=@n13, B=ScalarParameter() {defaultValue=0.5, parameterName=Fill Strength})
  n15 = TextureSample(UVs=Panner(Coordinate=TextureCoordinate() {uTiling=0.1, vTiling=0.1}) {speedX=0.1, speedY=0.2}) {samplerType=SAMPLERTYPE_LinearColor, texture=/Skyfire/Cosmetics/Effects/EnergyGrid/Textures/HighContrastNoise_Plugin}
  n16 = AppendVector(A=@n15.R, B=@n15.R)
  n17 = Panner(Coordinate=Add(A=TextureCoordinate() {uTiling=1.5, vTiling=1.5}, B=Multiply(A=@n16) {constB=0.03})) {speedX=0.15, speedY=0.075}
  n18 = Multiply(
   A=@n13
   B=Multiply(
    A=TextureSample(UVs=@n17) {texture=/Skyfire/Cosmetics/Effects/EnergyGrid/Textures/DiffusedStylizedGrid_Plugin}.R
    B=ScalarParameter() {defaultValue=0.3, parameterName=Grid Strength}
   )
  )
  n19 = TextureSample() {texture=/Skyfire/Cosmetics/Effects/EnergyGrid/Textures/T_LockDeviceRandomDots_Plugin}
  n20 = Add(A=Multiply(A=Time(), B=ScalarParameter() {defaultValue=0.6, parameterName=Dot Speed}), B=@n19.R)
  n21 = Power(Base=Saturate(None=Cosine(None=@n20)))
  n22 = Ceil(None=@n19.R)
  n23 = Multiply(A=Saturate(None=Multiply(A=@n21, B=@n22)), B=Multiply(A=@n19.G, B=@n19.G))
  n24 = ScalarParameter() {defaultValue=5, parameterName=Dots Strength}
  n25 = Multiply(A=OneMinus(None=Fresnel(ExponentIn=@n10) {baseReflectFraction=0}), B=Add(A=Add(A=@n14, B=@n18), B=Multiply(A=@n23, B=@n24)))
  n26 = ScalarParameter() {defaultValue=4, group=Grid, parameterName=Fresnel Exponent}
  Multiply(A=Multiply(A=@n25, B=@n3), B=OneMinus(None=Fresnel(ExponentIn=@n26) {baseReflectFraction=0}))
  n27 = VectorParameter() {group=Grid, parameterName=Grid Color, defaultValue=<rgba 0 0.527 1 10>}
  Multiply(A=Multiply(A=@n27.RGB, B=@n27.RGB), B=@n25)
 ]
}
