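# Translucent unlit material for the fill of the Skyfire EnergyGrid cosmetic effect, drawing a soft-edged circular shape.
# A SphereMask centered in UV space, with adjustable radius and hardness, is multiplied by an Opacity parameter to form the disc.
# A world position offset term adds a vector offset to a scale about the actor position, driven by WPO Size Adjust.
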Material /Skyfire/Cosmetics/Effects/EnergyGrid/Materials/M_Skyfire_EnergyGrid_Fill_01
graph {
 properties {blendMode=BLEND_Translucent, pixelDepthOffsetMode=PDOM_Legacy, shadingModel=MSM_Unlit, usedShadingModels=Unlit}
 nodes [
  n1 = VectorParameter() {group=WPO, parameterName=WPO Offset, channelNames {a="", b=Z, g=Y, r=X}, defaultValue=<rgba -500 -150.0001 -100 0>}
  n2 = ActorPositionWS()
  n3 = WorldPosition()
  n4 = ScalarParameter() {defaultValue=1, group=WPO, parameterName=WPO Size Adjust}
  Add(A=@n1, B=Multiply(A=Subtract(A=@n2, B=@n3.XYZ), B=OneMinus(None=@n4)))
  Constant()
  n5 = Constant() {r=0.5}
  Multiply(
   A=SphereMask(
    A=TextureCoordinate()
    B=AppendVector(A=@n5, B=@n5)
    Radius=ScalarParameter() {defaultValue=0.5, parameterName=Radius}
    Hardness=ScalarParameter() {defaultValue=0.95, parameterName=Hardness}
   )
   B=ScalarParameter() {parameterName=Opacity}
  )
 ]
}
