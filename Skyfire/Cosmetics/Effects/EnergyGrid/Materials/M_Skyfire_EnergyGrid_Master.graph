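# Additive, unlit, two-sided master material for the Skyfire Energy Grid effect, combining a grid, distorted SDF glyph, chaotic clouds and glowing outer ring.
# Grid tiling switches by noise-dithered camera distance, with panning noise driving sine pulses and a Blueprint Impact parameter boosting intensity.
# Colors shift with camera or lowest-player height relative to the actor.
# Scene-depth intersection feeds the ring, depth fade softens opacity, and WPO offsets and resizes the mesh.
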Material /Skyfire/Cosmetics/Effects/EnergyGrid/Materials/M_Skyfire_EnergyGrid_Master
graph {
 properties {blendMode=BLEND_Additive, pixelDepthOffsetMode=PDOM_Legacy, shadingModel=MSM_Unlit, twoSided=true, usedShadingModels=Unlit}
 nodes [
  n1 = VectorParameter() {group=WPO, parameterName=WPO Offset, channelNames {a="", b=Z, g=Y, r=X}, defaultValue=<rgba 0 0 -120 0>}
  n2 = ActorPositionWS()
  n3 = WorldPosition()
  n4 = ScalarParameter() {defaultValue=1, group=WPO, parameterName=WPO Size Adjust}
  Add(A=@n1, B=Multiply(A=Subtract(A=@n2, B=@n3.XYZ), B=OneMinus(None=@n4)))
  n5 = VectorParameter() {group=Outer Ring, parameterName=Radius, channelNames {a=D, b=C, g=B, r=A}, defaultValue=<rgba 0.5 0.47 0.48 0.6429>}
  n6 = MaterialFunctionCall(Float4=Multiply(A=ScalarParameter() {defaultValue=1, group=Outer Ring, parameterName=Master Radius}, B=AppendVector(A=@n5, B=@n5))) {materialFunction=&BreakOutFloat4Components}
  n7 = VectorParameter() {group=Outer Ring, parameterName=Density, channelNames {a=D, b=C, g=B, r=A}, defaultValue=<rgba 5.7251 60.3749 110.5775 5.8157>}
  Clamp(None=Subtract(
    A=MaterialFunctionCall(Radius=@n6.R, Density=@n7.A) {materialFunction=&RadialGradientExponential}.RadialGradientExponential
    B=MaterialFunctionCall(Radius=@n6.G, Density=@n7.B) {materialFunction=&RadialGradientExponential}.RadialGradientExponential
   ))
  n8 = TextureCoordinate()
  n9 = TextureObjectParameter() {
   group=Clouds
   parameterName=Clouds Texture
   samplerType=SAMPLERTYPE_LinearColor
   texture=/Game/VisualThreatManager/StormVisuals/Textures/Clouds_VectorDisplacement
  }
  n10 = MaterialFunctionCall!(
   Coordinates=Multiply(A=TextureCoordinate(), B=ScalarParameter() {defaultValue=1, group=Clouds, parameterName=Clouds Tile})
   Speed=ScalarParameter() {defaultValue=0.1, group=Clouds, parameterName=Clouds Speed}
   Divisor
   Texture=@n9
  ) {materialFunction=/Engine/Functions/Engine_MaterialFunctions03/ComplexMotion/Motion_4WayChaos}
  n11 = Multiply(
   A=Subtract(A=ComponentMask(None=@n10.Result) {g=true, r=true}) {constB=0.7}
   B=ScalarParameter() {defaultValue=0.015, group=Glyph, parameterName=Glyph distortion Str}
  ) {constB=0.01}
  n12 = TextureCoordinate()
  n13 = ScalarParameter() {defaultValue=10, group=Clouds, parameterName=CloudsB Tile}
  n14 = TextureObjectParameter() {
   group=Clouds
   parameterName=Clouds Texture
   samplerType=SAMPLERTYPE_LinearColor
   texture=&Clouds_VectorDisplacement
  }
  n15 = ComponentMask(None=MaterialFunctionCall(Coordinates=Multiply(A=@n12, B=@n13), Speed=ScalarParameter() {defaultValue=0.1, group=Clouds, parameterName=CloudsB Speed}, Texture=@n14) {materialFunction=&Motion_4WayChaos}.Result) {g=true, r=true}
  n16 = Add(
   A=MaterialFunctionCall(UVs=Add(A=@n8, B=@n11), Texture Scale=ScalarParameter() {defaultValue=1.05, group=Glyph, parameterName=Glyphe Scale Size}) {materialFunction=&ScaleUVsByCenter}.UVs
   B=Multiply(A=Subtract(A=@n15) {constB=0.7}, B=ScalarParameter() {defaultValue=0.006, group=Glyph, parameterName=Glyph Distortion Str B}) {constB=0.01}
  ) {constB=0}
  n17 = TextureSampleParameter2D(UVs=BumpOffset(Coordinate=Saturate(None=@n16), Height=ScalarParameter() {defaultValue=-0.6, group=Glyph, parameterName=Glyph Bump Offset}) {referencePlane=0}) {
   group=Glyph
   parameterName=Glyph Texture
   samplerSource=SSM_Clamp_WorldGroupSettings
   samplerType=SAMPLERTYPE_LinearGrayscale
   texture=/Skyfire/Cosmetics/Effects/EnergyGrid/VFX/T_Skyfire_EnergyGrid_FX
  }
  n18 = Round(None=Add(A=@n17.R, B=ScalarParameter() {defaultValue=0.285, group=Glyph, parameterName=Glyph SDF Offset}))
  n19 = ScalarParameter() {defaultValue=0.05, group=Glyph, parameterName=Glyph opacity add}
  n20 = ComponentMask(None=@n10.Result) {r=true}
  n21 = Power(Base=Multiply(A=Subtract(A=@n20) {constB=0.5}) {constB=3}, Exp=ScalarParameter() {defaultValue=3, group=Clouds, parameterName=Clouds EXP})
  n22 = LinearInterpolate(
   A=ScalarParameter() {defaultValue=2, group=Fill, parameterName=Fill_Strength_Near}
   B=ScalarParameter() {defaultValue=0.5, group=Fill, parameterName=Fill_Strength_Far}
   Alpha=Power(Base=NamedRerouteUsage())
  )
  n23 = SphereMask(A=TextureCoordinate(), B=Constant() {r=0.5}) {attenuationRadius=0.5, hardnessPercent=0}
  n24 = OneMinus(None=Power(Base=OneMinus(None=@n23), Exp=ScalarParameter() {defaultValue=3, group=Grid, parameterName=Edge Fade Exp}) {constExponent=4})
  n25 = AppendVector(
   A=ScalarParameter() {defaultValue=10, group=Grid UV, parameterName=X-Tiling Far}
   B=ScalarParameter() {defaultValue=10, group=Grid UV, parameterName=Y-Tiling Far}
  )
  n26 = AppendVector(
   A=ScalarParameter() {defaultValue=50, group=Grid UV, parameterName=X-Tiling Near}
   B=ScalarParameter() {defaultValue=50, group=Grid UV, parameterName=Y-Tiling Near}
  )
  n27 = SphereMask(
   A=WorldPosition().XYZ
   B=CameraPositionWS()
   Radius=ScalarParameter() {defaultValue=3000, group=Grid UV, parameterName=Near Camera Fade Grid Tile Radius}
   Hardness=ScalarParameter() {defaultValue=0.5, group=Grid UV, parameterName=Near Camera Fade Grid Tile Hardness}
  ) {hardnessPercent=1}
  n28 = TextureSample(UVs=Multiply(A=TextureCoordinate(), B=ScalarParameter() {defaultValue=25, group=Grid UV, parameterName=Noise Tile})) {samplerType=SAMPLERTYPE_LinearGrayscale, texture=&/modifiedGood64x64TilingNoiseHighFreq}
  n29 = Subtract(A=@n28.R) {constB=0.5}
  n30 = Saturate(None=Round(None=Add(A=@n27, B=@n29)))
  n31 = TextureSampleParameter2D(UVs=Multiply(A=TextureCoordinate(), B=LinearInterpolate(A=@n25, B=@n26, Alpha=@n30))) {
   group=Grid UV
   parameterName=Grid UV Texture
   samplerType=SAMPLERTYPE_LinearColor
   texture=/Game/Environments/Apollo/Sets/Skyfire/Textures/T_Skyfire_PackA
  }
  n32 = ScalarParameter() {defaultValue=3, group=Grid UV, parameterName=Pan Variation Sphere mask Amount}
  n33 = Multiply(A=TextureCoordinate(), B=ScalarParameter() {defaultValue=2, group=Grid UV, parameterName=Pan texture Variation Tiling})
  n34 = Multiply(A=Time(), B=ScalarParameter() {group=Grid UV, parameterName=Pan texture Variation Speed})
  n35 = Panner(Coordinate=@n33, Time=@n34, Speed=Constant2Vector() {g=1, r=1})
  n36 = TextureObjectParameter() {group=Grid UV, parameterName=Pan Variation Texture, texture=/Skyfire/Cosmetics/Effects/EnergyGrid/VFX/Pixely_noise}
  n37 = Panner(Coordinate=@n33, Time=@n34, Speed=Constant2Vector() {g=-1, r=-1})
  n38 = Divide(A=Add(
    A=TextureSample(UVs=@n35, Tex=@n36) {samplerType=SAMPLERTYPE_LinearColor, texture=&HighContrastNoise}.R
    B=TextureSample(UVs=@n37, Tex=@n36) {samplerType=SAMPLERTYPE_LinearColor, texture=&HighContrastNoise}.G
   ))
  n39 = ScalarParameter() {defaultValue=3, group=Grid UV, parameterName=Pan Variation Amount}
  n40 = Add(
   A=Add(A=Multiply(A=@n23, B=@n32), B=Multiply(A=@n38, B=@n39))
   B=Multiply(A=Time(), B=ScalarParameter() {defaultValue=0.15, group=Grid UV, parameterName=Pan Variation Speed})
  )
  n41 = Sine(None=Add(A=Multiply(A=@n40), B=OneMinus(None=@n31.B)))
  n42 = SphereMask(
   A=WorldPosition().XYZ
   B=CameraPositionWS()
   Radius=ScalarParameter() {defaultValue=1000, group=Grid, parameterName=Near Camera Fade Distance}
  ) {hardnessPercent=1}
  n43 = ScalarParameter() {defaultValue=8, group=Grid, parameterName=Near Camera Fade Distance exp}
  n44 = NamedRerouteDeclaration(None=OneMinus(None=Power(Base=@n42, Exp=@n43))) {name=Near Camera Fade, nodeColor=<rgba 0 1 0.3203 1>}
  n45 = LinearInterpolate(
   A=ScalarParameter() {defaultValue=3, group=Grid, parameterName=Intensity-Grid_Near}
   B=ScalarParameter() {defaultValue=1, group=Grid, parameterName=Intensity-Grid_Far}
   Alpha=@n44
  )
  n46 = LinearInterpolate(
   A=ScalarParameter() {defaultValue=10, group=Grid, parameterName=Intensity-Grid_Near_Glyph}
   B=ScalarParameter() {defaultValue=1.25, group=Grid, parameterName=Intensity-Grid_Far_glyph}
   Alpha=@n44
  )
  n47 = NamedRerouteUsage()
  n48 = Multiply(
   A=LinearInterpolate(
    A=ScalarParameter() {group=ForBlueprint, parameterName=Impact, sliderMax=1}
    B=ScalarParameter() {parameterName=Debug Impact Value}
    Alpha=ScalarParameter() {parameterName=Use Debug Impact Value}
   )
   B=ScalarParameter() {defaultValue=6, group=Impact, parameterName=Impact_Intensity}
  )
  n49 = Constant() {r=1}
  n50 = Multiply(
   A=LinearInterpolate(
    A=ScalarParameter() {defaultValue=0.25, group=Grid, parameterName=Grid Pan Low}
    B=ScalarParameter() {defaultValue=0.65, group=Grid, parameterName=Grid pan High}
    Alpha=ConstantBiasScale(None=@n41)
   )
   B=Multiply(A=Add(A=@n48, B=@n49), B=LinearInterpolate(A=@n45, B=@n46, Alpha=@n47))
  )
  n51 = Multiply(
   A=Add(A=Multiply(A=@n24, B=@n22), B=Multiply(A=@n24, B=@n50))
   B=ScalarParameter() {defaultValue=0.15, group=Grid, parameterName=Grid Opacity}
  )
  n52 = Multiply(
   A=@n51
   B=LinearInterpolate(
    A=ScalarParameter() {defaultValue=1.6, group=Clouds, parameterName=Clouds value High}
    B=ScalarParameter() {defaultValue=0.5, group=Clouds, parameterName=Clouds Value Low}
    Alpha=@n21
   )
  )
  n53 = ScalarParameter() {defaultValue=100, group=Outer Ring, parameterName=Intersection_Distance Divisor, sortPriority=33}
  n54 = MaterialFunctionCall!(
   Distance Bias
   Distance Divisor=LinearInterpolate(A=Multiply(A=@n53) {constB=0.3}, B=@n53, Alpha=NamedRerouteUsage())
   Scene Depth
   Pixel Depth
  ) {materialFunction=/Game/Effects/Fort_Effects/Materials/_Functions/MF_DistanceBehindTranslucency_01}
  n55 = ScalarParameter() {defaultValue=3, group=Outer Ring, parameterName=Intersection_Exp, sortPriority=33}
  n56 = Subtract(
   A=MaterialFunctionCall(Radius=@n6.B, Density=@n7.C) {materialFunction=&RadialGradientExponential}.RadialGradientExponential
   B=MaterialFunctionCall(Radius=@n6.A, Density=@n7.D) {materialFunction=&RadialGradientExponential}.RadialGradientExponential
  )
  n57 = Power(
   Base=Add(A=Power(Base=@n54.Normalized Inverted, Exp=@n55), B=Clamp(None=@n56))
   Exp=ScalarParameter() {defaultValue=1.25, group=Outer Ring, parameterName=Ring Power}
  )
  n58 = VectorParameter() {group=Clouds, parameterName=Cloud Tint opacity, channelNames {a=opacity, b="", g="", r=""}, defaultValue=<rgba 0.252 0.07 0.35 0.25>}
  n59 = Add(A=Add(A=Multiply(A=@n18, B=@n19) {constB=0.05}, B=Add(A=@n52, B=@n57)), B=Multiply(A=@n21, B=@n58.opacity))
  n60 = VectorParameter() {group=Grid, parameterName=Grid Color Far Falling Player grid, defaultValue=<rgba 1 0.1 1 1>}
  n61 = VectorParameter() {group=Grid, parameterName=Grid Color Near Falling Player Grid, defaultValue=<rgba 0 0.1667 1 1>}
  n62 = LinearInterpolate(
   A=ComponentMask(None=CameraPositionWS()) {b=true}
   B=ScalarParameter() {defaultValue=46250.4492, group=ForBlueprint, parameterName=LowestPlayerHeightWS}
   Alpha=ScalarParameter() {group=ForBlueprint, parameterName=UseLowestPlayerHeight}
  )
  n63 = ComponentMask(None=ActorPositionWS()) {b=true}
  n64 = Saturate(None=Divide(A=Distance(A=@n62, B=@n63), B=ScalarParameter() {defaultValue=5000, parameterName=Distance divider}) {constB=4000})
  n65 = ScalarParameter() {defaultValue=1, parameterName=transitionExp}
  n66 = Power(
   Base=OneMinus(None=Power(Base=@n64, Exp=@n65) {constExponent=1.25})
   Exp=ScalarParameter() {defaultValue=1, parameterName=transitionExpinvert}
  ) {constExponent=1.25}
  n67 = VectorParameter() {group=Grid, parameterName=Grid Color Far Falling Player Glyph, defaultValue=<rgba 0.7274 0.1 1 1>}
  n68 = VectorParameter() {group=Grid, parameterName=Grid Color Near Falling Player Glyph, defaultValue=<rgba 0 4 5 1>}
  n69 = SphereMask(
   A=WorldPosition().XYZ
   B=CameraPositionWS()
   Radius=ScalarParameter() {defaultValue=1100, group=Grid, parameterName=Near Camera Fade Distance Color}
  ) {hardnessPercent=1}
  n70 = LinearInterpolate(
   A=VectorParameter() {group=Grid, parameterName=Grid Color Near, defaultValue=<rgba 0 1.6 2 1>}.RGB
   B=LinearInterpolate(A=LinearInterpolate(A=@n60.RGB, B=@n61.RGB, Alpha=@n66), B=LinearInterpolate(A=@n67.RGB, B=@n68.RGB, Alpha=@n66), Alpha=NamedRerouteUsage())
   Alpha=Power(Base=OneMinus(None=@n69), Exp=ScalarParameter() {defaultValue=2, group=Grid, parameterName=DistanceRadius_Color_Hardness}) {constExponent=10}
  )
  n71 = VectorParameter() {group=Outer Ring, parameterName=RingColor, defaultValue=<rgba 250 48.1075 17.0472 1>}
  n72 = Add(
   A=Multiply(A=@n18, B=VectorParameter() {group=Glyph, parameterName=Glyph Tint, defaultValue=<rgba 0.1175 0.027 0.2 1>}.RGB)
   B=Add(A=Multiply(A=@n70, B=@n51), B=Multiply(A=@n71.RGB, B=@n57))
  )
  n73 = Multiply(A=@n21, B=@n58)
  Multiply(A=Multiply(A=Add(A=@n72, B=@n73), B=@n59), B=ScalarParameter() {defaultValue=1, group=Base, parameterName=Master emissive})
  n74 = Clamp(None=Add(A=@n59, B=ScalarParameter() {defaultValue=0.8, group=Base, parameterName=Master Fill}))
  n75 = DepthFade(Opacity=@n74, FadeDistance=ScalarParameter() {defaultValue=1000, group=Base, parameterName=FadeDistance})
  Multiply(
   A=FeatureLevelSwitch(Default=@n75, ES2=@n74, ES3_1=@n74, SM4_REMOVED=@n75, SM5=@n75, SM6=@n75)
   B=ScalarParameter() {defaultValue=0.95, group=Base, parameterName=Master Opacity}
  )
  NamedRerouteDeclaration(None=@n18) {name=Glyph texture, nodeColor=<rgba 1 0 0.7969 1>}
  NamedRerouteDeclaration(None=@n50) {name=CloudNoise, nodeColor=<rgba 0 1 0.5547 1>}
 ]
}
