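# Niagara system for the Skyfire EnergyGrid cosmetic effect, built from three sprite emitters with facing set to +Z.
# Glow spawns larger violet sprites by rate, nudged by a camera offset.
# Portal and Portal1 burst one counter-rotating sprite per loop, with Portal's first loop delayed so the pair staggers.
# All three fade in and out through an alpha curve over particle life, and the system exposes no user parameters.
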NiagaraSystem /Skyfire/Cosmetics/Effects/EnergyGrid/VFX/NS_Skyfire_EnergyGrid
dependencies {usedDataInterfaces [&NiagaraDataInterfaceCurve], usedDynamicInputs [], usedModules [&SystemState]}
emitters {
 Glow {
  emitterUpdateScript [
   &EmitterState(Inactive Response=::NewEnumerator1, Life Cycle Mode=::NewEnumerator1, Loop Delay=2, Loop Duration=10, UseLoopDelay=false) {module=EmitterLifeCycle}
   &SpawnRate(Use Spawn Probability=false)
   &SpawnBurst_Instantaneous(Spawn Count=1, Use Spawn Probability=false) {enabled=false}
  ]
  particleSpawnScript [
   &InitializeParticle(
    Color=<rgba 0.3631 0.0664 1 0.8>
    Sprite Size Mode=::NewEnumerator3
    Uniform Sprite Size=300
    UsePositionOffset=false
    Write Material Random=true
   ) {module=InitializeParticle001}
   &FN_Sprite_Attributes(Create Facing Attribute=true, Facing:Vector3f=<0 0 1>, Lifetime=3)
   &CameraOffset(Camera Offset Amount=50)
  ]
  particleUpdateScript [
   &ParticleState()
   &SolveForcesAndVelocity(Clamp Velocity=false, Limit Acceleration=false)
   &SpriteRotationRate(Rotation Rate=&RandomRangeFloat(Minimum=-30, Maximum=-20)) {enabled=false}
   &ScaleColor(
    Scale Alpha=&FloatFromCurve(FloatCurve=DataInterface() {propertyValues {Curve {keys [<at 0 0>,<at 0.5 1>,<at 1 0>]}}})
    ScaleA=true
    ScaleRGB=true
   )
  ]
  rendererClasses [&NiagaraSpriteRendererProperties]
  renderers [{rendererClass=&NiagaraSpriteRendererProperties}]
 }
 Portal {
  emitterUpdateScript [
   &EmitterState(
    Delay First Loop Only=true
    Inactive Response=::NewEnumerator1
    Life Cycle Mode=::NewEnumerator1
    Loop Delay=2
    Loop Duration=4
    UseLoopDelay=true
   ) {module=EmitterLifeCycle}
   &SpawnRate(SpawnRate=0.25, Use Spawn Probability=false) {enabled=false}
   &SpawnBurst_Instantaneous(Spawn Count=1, Use Spawn Probability=false)
  ]
  particleSpawnScript [
   &InitializeParticle(Lifetime=3, Sprite Size Mode=::NewEnumerator3, Uniform Sprite Size=200, UsePositionOffset=false, Write Material Random=true) {module=InitializeParticle001}
   &FN_Sprite_Attributes(
    Create Facing Attribute=true
    Create Sprite Rotation Attribute=true
    Facing:Vector3f=<0 0 1>
    Lifetime=3
    SpriteRotation:NiagaraFloat=&MakeCustomFloatFromBool(Bool=&RandomBool(), True Float=0, False Float=1)
   )
  ]
  particleUpdateScript [
   &ParticleState()
   &SolveForcesAndVelocity(Clamp Velocity=false, Limit Acceleration=false)
   &SpriteRotationRate(Rotation Rate=&RandomRangeFloat(Minimum=-30, Maximum=-20))
   &ScaleColor(
    Scale Alpha=&FloatFromCurve(FloatCurve=DataInterface() {propertyValues {Curve {keys [<at 0 0>,<at 0.5 1>,<at 1 0>]}}})
    ScaleA=true
    ScaleRGB=true
   )
  ]
  rendererClasses [&NiagaraSpriteRendererProperties]
  renderers [{rendererClass=&NiagaraSpriteRendererProperties}]
 }
 Portal1 {
  emitterUpdateScript [
   &EmitterState(Inactive Response=::NewEnumerator1, Life Cycle Mode=::NewEnumerator1, Loop Duration=4, UseLoopDelay=false) {module=EmitterLifeCycle}
   &SpawnRate(SpawnRate=0.25, Use Spawn Probability=false) {enabled=false}
   &SpawnBurst_Instantaneous(Spawn Count=1, Use Spawn Probability=false)
  ]
  particleSpawnScript [
   &InitializeParticle(Lifetime=3, Sprite Size Mode=::NewEnumerator3, Uniform Sprite Size=200, UsePositionOffset=false, Write Material Random=true) {module=InitializeParticle001}
   &FN_Sprite_Attributes(
    Create Facing Attribute=true
    Create Sprite Rotation Attribute=true
    Facing:Vector3f=<0 0 1>
    Lifetime=3
    SpriteRotation:NiagaraFloat=&MakeCustomFloatFromBool(Bool=&RandomBool(), True Float=0, False Float=1)
   )
  ]
  particleUpdateScript [
   &ParticleState()
   &SolveForcesAndVelocity(Clamp Velocity=false, Limit Acceleration=false)
   &SpriteRotationRate(Rotation Rate=&RandomRangeFloat(Minimum=30, Maximum=20))
   &ScaleColor(
    Scale Alpha=&FloatFromCurve(FloatCurve=DataInterface() {propertyValues {Curve {keys [<at 0 0>,<at 0.5 1>,<at 1 0>]}}})
    ScaleA=true
    ScaleRGB=true
   )
  ]
  rendererClasses [&NiagaraSpriteRendererProperties]
  renderers [{rendererClass=&NiagaraSpriteRendererProperties}]
 }
}
summary {
 systemName=NS_Skyfire_EnergyGrid
 emitters [
  {emitterName=Portal1, rendererClasses [&NiagaraSpriteRendererProperties]}
  {emitterName=Portal, rendererClasses [&NiagaraSpriteRendererProperties]}
  {emitterName=Glow, rendererClasses [&NiagaraSpriteRendererProperties]}
 ]
 userVariables []
}
